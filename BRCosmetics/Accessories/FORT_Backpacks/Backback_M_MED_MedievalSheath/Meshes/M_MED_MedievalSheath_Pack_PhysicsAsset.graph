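# Physics asset for the Medieval Sheath backpack cosmetic mesh, a Battle Royale back accessory.
# It holds a single body on the pack bone, made of one offset, rotated capsule that approximates the pack's volume.
# It defines no constraints, so there is no jointed chain or ragdoll setup between bones.
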PhysicsAsset /BRCosmetics/Accessories/FORT_Backpacks/Backback_M_MED_MedievalSheath/Meshes/M_MED_MedievalSheath_Pack_PhysicsAsset
bodies {pack [{shapeType=Capsule, length=23.3093, radius=20.0172, center=<0 -8.1648 -5.4203>, rotation=<rot -0.0899 -90 0>}]}
constraints []
